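# Sound cue for the Teriyaki back bling, using the reactive back bling property sheet, the Fort_Backblings sound class and close-range pet attenuation.
# After a short fixed delay, nested random nodes pick one of three waves, so the first plays about half the time and the other two a quarter each.
# Overridden concurrency allows one instance per owner and blocks new plays while one is active.
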SoundCue /BRCosmetics/Sounds/Backblings/Teriyaki/Teriyaki_Backpack_Cue
root=Delay_0
nodes [
 {id=Delay_0, children [Random_3], properties {delayMax=0.35, delayMin=0.35}}
 {id=Random_2, children [WavePlayer_4,WavePlayer_5], properties {weights [1,1]}}
 {id=Random_3, children [WavePlayer_6,Random_2], properties {weights [1,1]}}
 {id=WavePlayer_4, properties {soundWaveAssetPtr=&/Teriyaki_Backpack_02}}
 {id=WavePlayer_5, properties {soundWaveAssetPtr=&/Teriyaki_Backpack_03}}
 {id=WavePlayer_6, properties {soundWaveAssetPtr=&/Teriyaki_Backpack_01}}
]
properties {
 attenuationSettings=&Fort_Pets_Close_Athena_Att
 audioPropertiesSheet=/BRCosmetics/Sounds/Settings/PropertySheets/Reactive/APS_Cosmetic_Reactive_Backbling
 bOverrideConcurrency=true
 maxDistance=1800.0001
 soundClassObject=/Game/Sounds/FortSoundClassesAndModes/Fort_Backblings
 volumeMultiplier=0.8
 concurrencyOverrides {bLimitToOwner=true, maxCount=1, resolutionRule=PreventNew}
}
